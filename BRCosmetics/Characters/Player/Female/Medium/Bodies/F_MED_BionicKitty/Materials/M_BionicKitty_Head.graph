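# Character head material for the BionicKitty female medium body, built on the 2023 character parent material functions with material attributes.
# The PreFX parent function feeds a flipbook face function for mouth, eye and brow, then a generic visor function uses that result with the Tech Art Mask.
# The visor also returns UVs to override the parent's basic emissive UVs, and PostFX finishes the material.
# It supports skeletal meshes, morph targets and mesh deformers.
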Material /BRCosmetics/Characters/Player/Female/Medium/Bodies/F_MED_BionicKitty/Materials/M_BionicKitty_Head
graph {
 properties {
  bUseMaterialAttributes=true
  bUsedWithMeshDeformer=true
  bUsedWithMorphTargets=true
  bUsedWithSkeletalMesh=true
  pixelDepthOffsetMode=PDOM_Legacy
  shadingModel=MSM_FromMaterialExpression
  usedShadingModels=Default Lit | Subsurface | Clear Coat | Subsurface Profile | Cloth
 }
 nodes [
  n1 = MaterialFunctionCall(Override Basic Emissive UVs=@n2.UVs) {materialFunction=&MF_CharacterParent_2023_PreFX}
  n2 = MaterialFunctionCall!(
   Material=MaterialFunctionCall!(Mouth UV, Eye UV, Brow UV, Material=@n1.Result) {materialFunction=/Game/Global/Materials/Character_Parent_2023/FlipbookFace_Parent_2023/MaterialFunctions/MF_FlipbookFace}.Material
   Tech Art Mask=@n1.Tech Art Mask
  ) {materialFunction=/Game/Global/Materials/Character_Parent_2023/Visor_Parent_2025/MF_GenericVisor}
  MaterialFunctionCall(Material=@n2.Result, M=@n1.M, Pre Skinned Local Position=@n1.M, Distance Blend=@n1.M, Fresnel=@n1.M) {materialFunction=&MF_CharacterParent_2023_PostFX}
 ]
}
