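# Animation Blueprint for the female BionicKitty body cosmetic, which copies its attached parent's pose and corrects head, neck, headphones and tail.
# Each update, head-to-neck_02 yaw drives a headphones push-back alpha and, outside emotes, a head-limit alpha.
# These feed additive head/neck roll corrections and a Control Rig, followed by a weighted Tail_root constraint.
# Headphone constraints, extra bone transforms and a counter-rotation event chain are left unconnected.
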AnimBlueprint /BRCosmetics/Characters/Player/Female/Medium/Bodies/F_MED_BionicKitty/Meshes/F_MED_BionicKitty_AnimBP
target_skeleton=&/F_MED_BionicKitty_Skeleton
graphs [
 AnimGraph {nodes [
   n1 = "Transform (Modify) Bone - Bone: deform_headphones_base_m"(Alpha=1, Translation="0,-5.5,0", Rotation="0,0,-17", Scale="1,1,1") {
    rotationMode=BMM_Additive
    translationMode=BMM_Additive
    alphaBoolBlend {blendOption=Linear}
    boneToModify {boneName=deform_headphones_base_m}
    rotation {roll=-17}
    translation {y=-5.5}
   }
   "Constraint - deform_headphones_base_m (T)"(ComponentPose=@n1, Alpha=1, ConstraintWeights_0=1) {
    alphaBoolBlend {blendOption=Linear}
    boneToModify {boneName=deform_headphones_base_m}
    constraintSetup [{transformType=Translation, targetBone {boneName=deform_collar_base_m}}]
    constraintWeights [1]
   }
   n2 = "Constraint - deform_headphones_base_r (PPPPPPPP)"!(
    ComponentPose
    Alpha=1
    ConstraintWeights_0=0.1442
    ConstraintWeights_1:float=0.0683
    ConstraintWeights_2:float=0.0297
    ConstraintWeights_3:float=0.2462
    ConstraintWeights_4:float=0.0458
    ConstraintWeights_5:float=0.0402
    ConstraintWeights_6:float=0.0174
    ConstraintWeights_7:float=0.4082
   ) -> (Pose) {
    alphaBoolBlend {blendOption=Linear}
    boneToModify {boneName=deform_headphones_base_r}
    constraintSetup [
     {targetBone {boneName=spine_05}}
     {targetBone {boneName=clavicle_r}}
     {targetBone {boneName=neck_01}}
     {targetBone {boneName=head}}
     {targetBone {boneName=neck_02}}
     {targetBone {boneName=deform_clavicle_r}}
     {targetBone {boneName=deform_collar_base_fr}}
     {targetBone {boneName=deform_collar_r}}
    ]
    constraintWeights [1,1,1,1,1,1,1,1]
   }
   n3 = "Constraint - deform_headphones_base_r_bk (PPPPPPPP)"!(
    ComponentPose=@n2
    Alpha=1
    ConstraintWeights_0=0.2183
    ConstraintWeights_1:float=0.0643
    ConstraintWeights_2:float=0.0376
    ConstraintWeights_3:float=0.121
    ConstraintWeights_4:float=0.0705
    ConstraintWeights_5:float=0.0095
    ConstraintWeights_6:float=0.2731
    ConstraintWeights_7:float=0.2055
   ) -> (Pose) {
    alphaBoolBlend {blendOption=Linear}
    boneToModify {boneName=deform_headphones_base_r_bk}
    constraintSetup [
     {targetBone {boneName=spine_05}}
     {targetBone {boneName=clavicle_r}}
     {targetBone {boneName=neck_01}}
     {targetBone {boneName=head}}
     {targetBone {boneName=neck_02}}
     {targetBone {boneName=deform_clavicle_r}}
     {targetBone {boneName=deform_collar_base_bk}}
     {targetBone {boneName=deform_collar_r}}
    ]
    constraintWeights [1,1,1,1,1,1,1,1]
   }
   n4 = "Constraint - deform_headphones_base_bk (PPPPPPPP)"!(
    ComponentPose=@n3
    Alpha=1
    ConstraintWeights_0=0.0138
    ConstraintWeights_1:float=0.2212
    ConstraintWeights_2:float=0.0016
    ConstraintWeights_3:float=0.0016
    ConstraintWeights_4:float=0.0613
    ConstraintWeights_5:float=0.0767
    ConstraintWeights_6:float=0.1065
    ConstraintWeights_7:float=0.5174
   ) -> (Pose) {
    alphaBoolBlend {blendOption=Linear}
    boneToModify {boneName=deform_headphones_base_bk}
    constraintSetup [
     {targetBone {boneName=spine_03}}
     {targetBone {boneName=spine_05}}
     {targetBone {boneName=clavicle_l}}
     {targetBone {boneName=clavicle_r}}
     {targetBone {boneName=neck_01}}
     {targetBone {boneName=head}}
     {targetBone {boneName=neck_02}}
     {targetBone {boneName=deform_collar_base_bk}}
    ]
    constraintWeights [1,1,1,1,1,1,1,1]
   }
   n5 = "Constraint - deform_headphones_base_l_bk (PPPPPPPP)"!(
    ComponentPose=@n4
    Alpha=1
    ConstraintWeights_0=0.0082
    ConstraintWeights_1:float=0.2103
    ConstraintWeights_2:float=0.0619
    ConstraintWeights_3:float=0.0369
    ConstraintWeights_4:float=0.1257
    ConstraintWeights_5:float=0.0703
    ConstraintWeights_6:float=0.2626
    ConstraintWeights_7:float=0.2239
   ) -> (Pose) {
    alphaBoolBlend {blendOption=Linear}
    boneToModify {boneName=deform_headphones_base_l_bk}
    constraintSetup [
     {targetBone {boneName=spine_03}}
     {targetBone {boneName=spine_05}}
     {targetBone {boneName=clavicle_l}}
     {targetBone {boneName=neck_01}}
     {targetBone {boneName=head}}
     {targetBone {boneName=neck_02}}
     {targetBone {boneName=deform_collar_base_bk}}
     {targetBone {boneName=deform_collar_l}}
    ]
    constraintWeights [1,1,1,1,1,1,1,1]
   }
   "Constraint - deform_headphones_base_l (PPPPPPPP)"!(
    ComponentPose=@n5
    Alpha=1
    ConstraintWeights_0=0.1414
    ConstraintWeights_1:float=0.0641
    ConstraintWeights_2:float=0.0296
    ConstraintWeights_3:float=0.254
    ConstraintWeights_4:float=0.0455
    ConstraintWeights_5:float=0.0371
    ConstraintWeights_6:float=0.0184
    ConstraintWeights_7:float=0.4099
   ) -> (Pose) {
    alphaBoolBlend {blendOption=Linear}
    boneToModify {boneName=deform_headphones_base_l}
    constraintSetup [
     {targetBone {boneName=spine_05}}
     {targetBone {boneName=clavicle_l}}
     {targetBone {boneName=neck_01}}
     {targetBone {boneName=head}}
     {targetBone {boneName=neck_02}}
     {targetBone {boneName=deform_clavicle_l}}
     {targetBone {boneName=deform_collar_base_fr}}
     {targetBone {boneName=deform_collar_l}}
    ]
    constraintWeights [1,1,1,1,1,1,1,1]
   }
   "Transform (Modify) Bone - Bone: deform_headphones_base_m"!(ComponentPose, bAlphaBoolEnabled:bool=True, Translation="0,0,0", Rotation="0,0,0", Scale="0,0,0") -> (Pose) {
    alphaInputType=Bool
    scaleMode=BMM_Additive
    alphaBoolBlend {blendOption=Linear}
    boneToModify {boneName=deform_headphones_base_m}
    scale=<0 0 0>
   }
   "Transform (Modify) Bone - Bone: head"(Alpha=1, Translation="0,0,0", Rotation="0,0,0", Scale="1,1,1") {rotationMode=BMM_Additive, alphaBoolBlend {blendOption=Linear}, boneToModify {boneName=head}}
   "Transform (Modify) Bone - Bone: neck_01"(
    ComponentPose="Transform (Modify) Bone - Bone: neck_02"(
     ComponentPose="Transform (Modify) Bone - Bone: head"(Alpha=1, Translation="0,0,0", Rotation="0,0,0", Scale="1,1,1") {rotationMode=BMM_Additive, alphaBoolBlend {blendOption=Linear}, boneToModify {boneName=head}}
     Alpha=1
     Translation="0,0,0"
     Rotation="0,0,0"
     Scale="1,1,1"
    ) {rotationMode=BMM_Additive, alphaBoolBlend {blendOption=Linear}, boneToModify {boneName=neck_02}}
    Alpha=1
    Translation="0,0,0"
    Rotation="0,0,0"
    Scale="1,1,1"
   ) {rotationMode=BMM_Additive, alphaBoolBlend {blendOption=Linear}, boneToModify {boneName=neck_01}}
   n6 = Get HeadLimitAlpha() -> (HeadLimitAlpha:double)
   n7 = "Transform (Modify) Bone - Bone: head"(
    ComponentPose=Local To Component(LocalPose=Copy Pose From Mesh() {bUseAttachedParent=true})
    Alpha=@n6
    Translation="0,0,0"
    Rotation="0,0,35"
    Scale="1,1,1"
   ) {
    rotationMode=BMM_Additive
    scaleMode=BMM_Additive
    translationMode=BMM_Additive
    alphaBoolBlend {blendOption=Linear}
    boneToModify {boneName=head}
    rotation {roll=35}
   }
   n8 = "Transform (Modify) Bone - Bone: neck_02"(ComponentPose=@n7, Alpha=@n6, Translation="0,0,0", Rotation="0,0,10", Scale="1,1,1") {
    rotationMode=BMM_Additive
    scaleMode=BMM_Additive
    translationMode=BMM_Additive
    alphaBoolBlend {blendOption=Linear}
    boneToModify {boneName=neck_02}
    rotation {roll=10}
   }
   n9 = "Transform (Modify) Bone - Bone: neck_01"(ComponentPose=@n8, Alpha=@n6, Translation="0,0,0", Rotation="0,0,5", Scale="1,1,1") {
    rotationMode=BMM_Additive
    scaleMode=BMM_Additive
    translationMode=BMM_Additive
    alphaBoolBlend {blendOption=Linear}
    boneToModify {boneName=neck_01}
    rotation {roll=5}
   }
   n10 = Control Rig!(
    Source=Component To Local(ComponentPose=@n9)
    Alpha
    HeadphonesMoveBack:double=Get HeadphonesMoveBackAlpha() -> (HeadphonesMoveBackAlpha:double)
   ) -> (Pose) {
    alphaBoolBlend {blendOption=Linear}
    controlRigAssetReference {blueprintRigClass=&/F_MED_BionicKitty_CtrlRig.F_MED_BionicKitty_CtrlRig_C}
   }
   n11 = "Constraint - Tail_root (PPPPPPP)"!(
    ComponentPose=Local To Component(LocalPose=@n10)
    Alpha=1
    ConstraintWeights_0=0.0317
    ConstraintWeights_1:float=0
    ConstraintWeights_2:float=0.0007
    ConstraintWeights_3:float=0.0192
    ConstraintWeights_4:float=0.9458
    ConstraintWeights_5:float=0.0013
    ConstraintWeights_6:float=0.0013
   ) -> (Pose) {
    alphaBoolBlend {blendOption=Linear}
    boneToModify {boneName=Tail_root}
    constraintSetup [
     {targetBone {boneName=spine_02}}
     {targetBone {boneName=spine_04}}
     {targetBone {boneName=spine_03}}
     {targetBone {boneName=spine_01}}
     {targetBone {boneName=pelvis}}
     {targetBone {boneName=deform_glute_l}}
     {targetBone {boneName=deform_glute_r}}
    ]
    constraintWeights [1,1,1,1,1,1,1]
   }
   Output Pose(Result=Component To Local(ComponentPose=@n11)) {layerGroup=None}
  ]}
 EventGraph {nodes [
   Event BlueprintUpdateAnimation()
   n1 = GetOwnerSkeletalMeshComponent()
   n2 = BreakRotator(InRot=BreakTransform(InTransform=GetDeltaTransformFromRefPose(self=@n1, BoneName=head, BaseName=neck_02)).Rotation)
   > Set HeadphonesMoveBackAlpha(HeadphonesMoveBackAlpha:double=MapRangeClamped(Value=@n2.Yaw, InRangeA=0, InRangeB=-40)) -> (Output_Get:double)
   > Branch(Condition=NOT Boolean(A=Get bIsPlayingEmote()))
   > Set HeadLimitAlpha(HeadLimitAlpha:double=MapRangeClamped(Value=@n2.Yaw, InRangeA=-10, InRangeB=-70)) -> (Output_Get:double)
   Branch(Condition=NOT Boolean(A=Get bIsPlayingEmote()))
   n3 = GetOwnerSkeletalMeshComponent()
   n4 = BreakRotator(InRot=BreakTransform(InTransform=GetDeltaTransformFromRefPose(self=@n3, BoneName=head, BaseName=neck_02)).Rotation)
   n5 = Get Z percentage() -> (Z percentage:float)
   n6 = Get X percentage() -> (X percentage:float)
   n7 = Get Y percentage() -> (Y percentage:float)
   > Set head rot(head rot:Rotator=MakeRotator(
     Roll=float * float!(A=@n4, B=-1, C:double=@n6) -> (ReturnValue)
     Pitch=float * float(A=@n4.Pitch, B=-1, C=@n7)
     Yaw=float * float(A=@n4.Yaw, B=-1, C=@n5)
    )) -> (Output_Get:Rotator)
   n8 = BreakRotator(InRot=BreakTransform(InTransform=GetDeltaTransformFromRefPose(self=@n3, BoneName=neck_02, BaseName=neck_01)).Rotation)
   n9 = Get Z percentage()
   n10 = Get X percentage()
   n11 = Get Y percentage()
   > Set neck_02_rot(neck_02_rot:Rotator=MakeRotator(Roll=float * float(A=@n8, B=-1, C=@n10), Pitch=float * float(A=@n8.Pitch, B=-1, C=@n11), Yaw=float * float(A=@n8.Yaw, B=-1, C=@n9))) -> (Output_Get:Rotator)
   n12 = BreakRotator(InRot=BreakTransform(InTransform=GetDeltaTransformFromRefPose(self=@n3, BoneName=neck_01, BaseName=spine_05)).Rotation)
   n13 = Get Y percentage()
   n14 = Get Z percentage()
   n15 = Get X percentage()
   > Set neck_01_rot(neck_01_rot:Rotator=MakeRotator(Roll=float * float(A=@n12, B=-1, C=@n15), Pitch=float * float(A=@n12.Pitch, B=-1, C=@n13), Yaw=float * float(A=@n12.Yaw, B=-1, C=@n14))) -> (Output_Get:Rotator)
  ]}
]
variables [
 head rot {type=Rotator}
 HeadLimitAlpha {type=number}
 HeadphonesMoveBackAlpha {type=number}
 neck_01_rot {type=Rotator}
 neck_02_rot {type=Rotator}
 X percentage {type=number, default=0.9}
 Y percentage {type=number, default=0.9}
 Z percentage {type=number, default=0.9}
]
