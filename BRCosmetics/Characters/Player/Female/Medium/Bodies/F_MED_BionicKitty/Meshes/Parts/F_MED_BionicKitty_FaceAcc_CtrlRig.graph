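# Control Rig for the BionicKitty character's face accessory that procedurally deforms headphone bones to follow body motion.
# The headphone base is parent-constrained across head, neck and upper spine, then spherical pose readers weight corrective transforms, including a combined head and right-clavicle corrective.
# Separate clavicle readers adjust the left and right sides.
# The emote branch only uses its Completed output, and several correctives, including the HeadphonesMoveBackAlpha one, are unconnected.
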ControlRigBlueprint /BRCosmetics/Characters/Player/Female/Medium/Bodies/F_MED_BionicKitty/Meshes/Parts/F_MED_BionicKitty_FaceAcc_CtrlRig
graphs [
 RigVMFunctionLibrary {links [], nodes []}
 RigVMFunction_Sequence_1_ContainedGraph {nodes [
   n1 = RigVMFunctionEntryNode()
   RigVMFunction_Sequence = &RigVMFunction_Sequence(ExecuteContext=@n1.ExecuteContext)
   n2 = &RigVMFunction_Sequence(ExecuteContext=@RigVMFunction_Sequence.B)
   RigVMFunctionReturnNode(A=@RigVMFunction_Sequence.A, B=@n2.A, C=@n2.B)
  ]}
 RigVMModel {nodes [
   RigUnit_BeginExecution = &RigUnit_BeginExecution()
   RigVMFunction_Sequence = RigVMAggregateNode(ExecuteContext=@RigUnit_BeginExecution.ExecutePin)
   &RigUnit_ParentConstraint(
    AdvancedSettings=(InterpolationType=Average,RotationOrderForFilter=XZY)
    Child=(Type=Bone,Name="deform_headphones_base_m")
    Parents=((Item=(Type=Bone,Name="head"),Weight=0.3),(Item=(Type=Bone,Name="neck_01"),Weight=0.3),(Item=(Type=Bone,Name="neck_02"),Weight=0.3),(Item=(Type=Bone,Name="spine_05"),Weight=0.1))
    Weight=1
    ExecutePin=@RigVMFunction_Sequence.A
   )
   > n1 = &RigUnit_SphericalPoseReader(
    ActiveRegionScaleFactors=(PositiveWidth=1,NegativeWidth=1,PositiveHeight=4.5,NegativeHeight=4.5)
    ActiveRegionSize=0.05
    Debug=(bDrawDebug=false,bDraw2D=False,bDrawLocalAxes=False,DebugScale=25,DebugSegments=20,DebugThickness=0.25)
    DriverAxis=(X=0,Y=1,Z=0)
    DriverItem=(Type=Bone,Name="deform_headphones_base_m")
    FalloffRegionScaleFactors=(PositiveWidth=0.2,NegativeWidth=0.18,PositiveHeight=1,NegativeHeight=1)
    FalloffSize=0.5
    FlipHeightScaling=False
    FlipWidthScaling=true
    OptionalParentItem=(Type=Bone,Name="spine_05")
    RotationOffset=(X=90,Y=0,Z=-26)
   )
   > n2 = &RigUnit_ModifyTransforms(
    ItemToModify="((Item=(Type=Bone,Name=\"deform_headphones_base_m\"),Transform=(Rotation=(X=0,Y=0,Z=0,W=1),Translation=(X=0,Y=0,Z=0),Scale3D=(X=1,Y=1,Z=1))),(Item=(Type=Bone,Name=\"deform_headphones_bk\"),Transform=(Rotation=(X=0,Y=0,Z=0,W=1),Translation=(X=0.8,Y=0,Z=0),Scale ...(truncated)"
    WeightMinimum=0
    Weight=@n1.OutputParam
   )
   RigVMFunction_ControlFlowBranch = &RigVMFunction_ControlFlowBranch(ExecuteContext=@n2.ExecutePin, Condition=RigVMVariableNode() {variable=isPlayingEmoteYouBoreMe}.Value)
   n3 = &RigUnit_SphericalPoseReader(
    ActiveRegionScaleFactors=(PositiveWidth=1,NegativeWidth=1,PositiveHeight=1,NegativeHeight=1)
    ActiveRegionSize=0.05
    Debug=(bDrawDebug=true,bDraw2D=False,bDrawLocalAxes=False,DebugScale=25,DebugSegments=20,DebugThickness=0.25)
    DriverAxis=(X=0,Y=1,Z=0)
    DriverItem=(Type=Bone,Name="head")
    FalloffRegionScaleFactors=(PositiveWidth=1,NegativeWidth=1,PositiveHeight=1,NegativeHeight=1)
    FalloffSize=0.1
    FlipHeightScaling=False
    FlipWidthScaling=true
    OptionalParentItem=(Type=Bone,Name="spine_05")
    RotationOffset=(X=95,Y=0,Z=-30)
    ExecutePin=@RigVMFunction_ControlFlowBranch.Completed
   )
   > n4 = &RigUnit_SphericalPoseReader(
    ActiveRegionScaleFactors=(PositiveWidth=1,NegativeWidth=1,PositiveHeight=1,NegativeHeight=1)
    ActiveRegionSize=0.1
    Debug=(bDrawDebug=true,bDraw2D=False,bDrawLocalAxes=False,DebugScale=25,DebugSegments=20,DebugThickness=0.25)
    DriverAxis=(X=1,Y=0,Z=0)
    DriverItem=(Type=Bone,Name="clavicle_r")
    FalloffRegionScaleFactors=(PositiveWidth=1,NegativeWidth=1,PositiveHeight=1,NegativeHeight=1)
    FalloffSize=0.2
    FlipHeightScaling=False
    FlipWidthScaling=False
    OptionalParentItem=(Type=Bone,Name="None")
    RotationOffset=(X=1.2435,Y=-45.7548,Z=4.5232)
   )
   Multiply = &RigVMFunction_MathDoubleMul(B=@n4.OutputParam, A=@n3.OutputParam)
   > &RigUnit_ModifyTransforms(
    ItemToModify="((Item=(Type=Bone,Name=\"deform_headphones_base_m\"),Transform=(Rotation=(X=0,Y=0.0491,Z=-0.0245,W=0.9985),Translation=(X=1.3005,Y=0.0126,Z=0.0252),Scale3D=(X=1,Y=1,Z=1))),(Item=(Type=Bone,Name=\"deform_headphones_bk\"),Transform=(Rotation=(X=0,Y=0,Z=0,W=1),Tr ...(truncated)"
    WeightMinimum=0
    Weight=@Multiply.Result
   )
   n5 = &RigUnit_SphericalPoseReader(
    ActiveRegionScaleFactors=(PositiveWidth=1,NegativeWidth=1,PositiveHeight=1,NegativeHeight=1)
    ActiveRegionSize=0.1
    Debug=(bDrawDebug=false,bDraw2D=False,bDrawLocalAxes=False,DebugScale=25,DebugSegments=20,DebugThickness=0.25)
    DriverAxis=(X=1,Y=0,Z=0)
    DriverItem=(Type=Bone,Name="clavicle_l")
    FalloffRegionScaleFactors=(PositiveWidth=1.5,NegativeWidth=1,PositiveHeight=1,NegativeHeight=1)
    FalloffSize=0.2
    FlipHeightScaling=False
    FlipWidthScaling=False
    OptionalParentItem=(Type=Bone,Name="None")
    RotationOffset=(X=0,Y=-36,Z=15)
    ExecutePin=@RigVMFunction_Sequence.B
   )
   > &RigUnit_ModifyTransforms(
    ItemToModify=((Item=(Type=Bone,Name="deform_headphones_l"),Transform=(Rotation=(X=-0.0072,Y=0.0548,Z=0.1303,W=0.9899),Translation=(X=-0.3881,Y=3.7009,Z=0.7741),Scale3D=(X=1,Y=1,Z=1))))
    WeightMinimum=0
    Weight=@n5.OutputParam
   )
   > n6 = &RigUnit_SphericalPoseReader(
    ActiveRegionScaleFactors=(PositiveWidth=1,NegativeWidth=1,PositiveHeight=1,NegativeHeight=1)
    ActiveRegionSize=0.1
    Debug=(bDrawDebug=false,bDraw2D=False,bDrawLocalAxes=False,DebugScale=25,DebugSegments=20,DebugThickness=0.25)
    DriverAxis=(X=1,Y=0,Z=0)
    DriverItem=(Type=Bone,Name="clavicle_r")
    FalloffRegionScaleFactors=(PositiveWidth=1.5,NegativeWidth=1,PositiveHeight=1,NegativeHeight=1)
    FalloffSize=0.2
    FlipHeightScaling=False
    FlipWidthScaling=False
    OptionalParentItem=(Type=Bone,Name="None")
    RotationOffset=(X=0,Y=-36,Z=15)
   )
   > &RigUnit_ModifyTransforms(
    ItemToModify=((Item=(Type=Bone,Name="deform_headphones_r"),Transform=(Rotation=(X=0.0072,Y=0.0548,Z=-0.1303,W=0.9899),Translation=(X=-0.2032,Y=-3.0107,Z=0.774),Scale3D=(X=1,Y=1,Z=1))))
    WeightMinimum=0
    Weight=@n6.OutputParam
   )
   n7 = &RigUnit_SphericalPoseReader(
    ActiveRegionScaleFactors=(PositiveWidth=1,NegativeWidth=1,PositiveHeight=4.5,NegativeHeight=4.5)
    ActiveRegionSize=0.05
    Debug=(bDrawDebug=true,bDraw2D=False,bDrawLocalAxes=False,DebugScale=25,DebugSegments=20,DebugThickness=0.25)
    DriverAxis=(X=0,Y=1,Z=0)
    DriverItem=(Type=Bone,Name="deform_headphones_base_m")
    ExecutePin=()
    FalloffRegionScaleFactors=(PositiveWidth=0.2,NegativeWidth=0.18,PositiveHeight=1,NegativeHeight=1)
    FalloffSize=0.5
    FlipHeightScaling=False
    FlipWidthScaling=true
    OptionalParentItem=(Type=Bone,Name="spine_05")
    RotationOffset=(X=90,Y=0,Z=25)
   )
   > &RigUnit_ModifyTransforms(
    ItemToModify="((Item=(Type=Bone,Name=\"deform_headphones_base_m\"),Transform=(Rotation=(X=0,Y=0,Z=-0.5,W=0.866),Translation=(X=0,Y=0,Z=0),Scale3D=(X=1,Y=1,Z=1))),(Item=(Type=Bone,Name=\"deform_headphones_bk\"),Transform=(Rotation=(X=0,Y=0,Z=0,W=1),Translation=(X=0,Y=0,Z=0), ...(truncated)"
    WeightMinimum=0
    Weight=@n7.OutputParam
   )
   n8 = &RigUnit_SphericalPoseReader(
    ActiveRegionScaleFactors=(PositiveWidth=1,NegativeWidth=1,PositiveHeight=1,NegativeHeight=1)
    ActiveRegionSize=0.1
    Debug=(bDrawDebug=false,bDraw2D=False,bDrawLocalAxes=False,DebugScale=25,DebugSegments=20,DebugThickness=0.25)
    DriverAxis=(X=0,Y=1,Z=0)
    DriverItem=(Type=Bone,Name="deform_headphones_base_m")
    ExecutePin=()
    FalloffRegionScaleFactors=(PositiveWidth=1,NegativeWidth=1,PositiveHeight=1.5,NegativeHeight=1)
    FalloffSize=0.2
    FlipHeightScaling=False
    FlipWidthScaling=False
    OptionalParentItem=(Type=Bone,Name="None")
    RotationOffset=(X=45,Y=0,Z=0)
   )
   > &RigUnit_ModifyTransforms(
    ItemToModify="((Item=(Type=Bone,Name=\"deform_headphones_base_m\"),Transform=(Rotation=(X=0,Y=0,Z=0,W=1),Translation=(X=0,Y=0,Z=0),Scale3D=(X=1,Y=1,Z=1))),(Item=(Type=Bone,Name=\"deform_headphones_bk\"),Transform=(Rotation=(X=0,Y=0,Z=0,W=1),Translation=(X=0,Y=0,Z=0),Scale3D ...(truncated)"
    WeightMinimum=0
    Weight=@n8.OutputParam
   )
   > n9 = &RigUnit_SphericalPoseReader(
    ActiveRegionScaleFactors=(PositiveWidth=1,NegativeWidth=1,PositiveHeight=1,NegativeHeight=1)
    ActiveRegionSize=0.1
    Debug=(bDrawDebug=false,bDraw2D=False,bDrawLocalAxes=False,DebugScale=25,DebugSegments=20,DebugThickness=0.25)
    DriverAxis=(X=0,Y=1,Z=0)
    DriverItem=(Type=Bone,Name="deform_headphones_base_m")
    FalloffRegionScaleFactors=(PositiveWidth=1,NegativeWidth=1,PositiveHeight=1,NegativeHeight=1.5)
    FalloffSize=0.2
    FlipHeightScaling=False
    FlipWidthScaling=False
    OptionalParentItem=(Type=Bone,Name="None")
    RotationOffset=(X=135,Y=0,Z=0)
   )
   > &RigUnit_ModifyTransforms(
    ItemToModify="((Item=(Type=Bone,Name=\"deform_headphones_base_m\"),Transform=(Rotation=(X=0,Y=0,Z=0,W=1),Translation=(X=0,Y=0,Z=0),Scale3D=(X=1,Y=1,Z=1))),(Item=(Type=Bone,Name=\"deform_headphones_bk\"),Transform=(Rotation=(X=0,Y=0,Z=0,W=1),Translation=(X=0,Y=0,Z=0),Scale3D ...(truncated)"
    WeightMinimum=0
    Weight=@n9.OutputParam
   )
   &RigUnit_GetRelativeTransformForItem(Child=(Type=Bone,Name="None"), Parent=(Type=Bone,Name="None"), bChildInitial=False, bParentInitial=False)
   &RigUnit_ModifyTransforms(
    ExecutePin=()
    ItemToModify="((Item=(Type=Bone,Name=\"deform_headphones_base_m\"),Transform=(Rotation=(X=0,Y=0,Z=-0.2588,W=0.9659),Translation=(X=-0.0862,Y=-2.8096,Z=0),Scale3D=(X=1,Y=1,Z=1))),(Item=(Type=Bone,Name=\"deform_headphones_bk\"),Transform=(Rotation=(X=0,Y=0,Z=0,W=-1),Translati ...(truncated)"
    Weight=1
    WeightMinimum=0
   )
   &RigUnit_ModifyTransforms(
    ExecutePin=()
    ItemToModify="((Item=(Type=Bone,Name=\"deform_headphones_bk\"),Transform=(Rotation=(X=0,Y=0,Z=0,W=1),Translation=(X=1,Y=0,Z=0),Scale3D=(X=1,Y=1,Z=1))),(Item=(Type=Bone,Name=\"deform_headphones_l_bk\"),Transform=(Rotation=(X=0,Y=0,Z=0,W=-1),Translation=(X=0.5,Y=0,Z=0),Scale3 ...(truncated)"
    WeightMinimum=0
    Weight=&RigVMFunction_MathDoubleRemap(SourceMaximum=0, SourceMinimum=1, TargetMaximum=1, TargetMinimum=0, Value=RigVMVariableNode() {variable=HeadphonesMoveBackAlpha}.Value).Result
   )
  ]}
]
hierarchy [
 root {loc=<0 0 0>, rot=<0 0 0 1>}
 pelvis {parent=root, loc=<0 -0.5301 98.0571>, rot=<0.0127 0.707 -0.0127 -0.707>}
 spine_01 {parent=pelvis, loc=<6.1208 -0.9328 0>, rot=<0 0 0.034 -0.9994>}
 spine_02 {parent=spine_01, loc=<6.5997 -0.2095 0>, rot=<0 0 0.0238 -0.9997>}
 spine_03 {parent=spine_02, loc=<6.4914 -0.1237 0>, rot=<0 0 0.0149 -0.9999>}
 spine_04 {parent=spine_03, loc=<6.9598 0.0794 0>, rot=<0 0 -0.0132 -0.9999>}
 spine_05 {parent=spine_04, loc=<6.9987 0 0>, rot=<0 0 -0.0015 -1>}
 neck_01 {parent=spine_05, loc=<11.3138 0.1529 0>, rot=<0 0 -0.2003 -0.9797>}
 neck_02 {parent=neck_01, loc=<4.5627 0 0>, rot=<0 0 0.0504 -0.9987>}
 head {parent=neck_02, loc=<5.3914 0 0>, rot=<0 0 0.115 -0.9934>}
 clavicle_l {parent=spine_05, loc=<5.4831 -0.1053 -3.2946>, rot=<0.0465 -0.7289 -0.009 -0.6829>}
 clavicle_r {parent=spine_05, loc=<5.4831 -0.1053 3.2946>, rot=<0.7289 0.0465 -0.6829 0.009>}
 deform_clavicle_base_l {parent=spine_05, loc=<10.8475 -1.9278 -6.5965>, rot=<0.0284 -0.7065 -0.0284 -0.7065>}
 deform_clavicle_l {parent=deform_clavicle_base_l, loc=<0 0 0>, rot=<0 0 0 1>}
 deform_clavicle_base_r {parent=spine_05, loc=<10.8474 -1.9279 6.5965>, rot=<0.7065 0.0284 0.7065 -0.0284>}
 deform_clavicle_r {parent=deform_clavicle_base_r, loc=<0 0 0>, rot=<0 0 0 -1>}
 deform_collar_base_m {parent=spine_05, loc=<14.4387 1.6926 0>, rot=<0 0 -0.1164 -0.9932>}
 deform_collar_base_l {parent=deform_collar_base_m, loc=<0 0 -8.1405>, rot=<0 -0.7071 0 -0.7071>}
 deform_collar_l {parent=deform_collar_base_l, loc=<0 0 0>, rot=<0 0 0 1>}
 deform_collar_base_r {parent=deform_collar_base_m, loc=<0.0001 0 8.1405>, rot=<0.7071 0 0.7071 0>}
 deform_collar_r {parent=deform_collar_base_r, loc=<0 0 0>, rot=<0 0 0 1>}
 deform_collar_base_fr {parent=deform_collar_base_m, loc=<-0.9604 8.6939 0>, rot=<0.4774 0.5216 0.5216 0.4774>}
 deform_collar_fr {parent=deform_collar_base_fr, loc=<0 0 0>, rot=<0 0 0 1>}
 deform_collar_base_bk {parent=deform_collar_base_m, loc=<-0.4571 -8.7041 -0.0005>, rot=<0.5 -0.5 0.5 -0.5>}
 deform_collar_bk {parent=deform_collar_base_bk, loc=<0 0 0>, rot=<0 0 0 1>}
 deform_headphones_base_m {parent=spine_05, loc=<14.4387 1.6926 0>, rot=<0 0 -0.1164 -0.9932>}
 deform_headphones_base_l {parent=deform_headphones_base_m, loc=<-0.4278 2.4741 -7.7104>, rot=<-0.1101 -0.7976 -0.1318 -0.5782>}
 deform_headphones_l {parent=deform_headphones_base_l, loc=<0 0 0>, rot=<0 0 0 1>}
 deform_headphones_base_bk {parent=deform_headphones_base_m, loc=<-0.4571 -8.7041 -0.0005>, rot=<0.5 -0.5 0.5 -0.5>}
 deform_headphones_bk {parent=deform_headphones_base_bk, loc=<0 0 0>, rot=<0 0 0 1>}
 deform_headphones_base_r {parent=deform_headphones_base_m, loc=<-0.4276 2.4741 7.7104>, rot=<0.5808 0.1512 0.7941 0.096>}
 deform_headphones_r {parent=deform_headphones_base_r, loc=<0 0 0>, rot=<0 0 0 -1>}
 deform_headphones_base_l_bk {parent=deform_headphones_base_m, loc=<-0.4848 -5.9204 -5.7616>, rot=<0.223 -0.7089 0.2867 -0.6046>}
 deform_headphones_l_bk {parent=deform_headphones_base_l_bk, loc=<0 0 0>, rot=<0 0 0 1>}
 deform_headphones_base_r_bk {parent=deform_headphones_base_m, loc=<-0.4975 -5.9556 5.797>, rot=<0.6312 -0.2994 0.6523 -0.2939>}
 deform_headphones_r_bk {parent=deform_headphones_base_r_bk, loc=<0 0 0>, rot=<0 0 0 1>}
]
variables [
 HeadphonesMoveBackAlpha {type=double, default=0}
 isPlayingEmoteYouBoreMe {type=bool, default=False}
]
