AnimBlueprint /BRCosmetics/Characters/Player/Female/Medium/Bodies/F_MED_BionicKitty/Meshes/Parts/F_MED_BionicKitty_FaceAcc_AnimBP
target_skeleton=&/F_MED_BionicKitty_FaceAcc_Skeleton
graphs [
 AnimGraph {nodes [
   n1 = "Constraint - deform_headphones_base_r (PPPPPPPPP)"!(
    ComponentPose
    Alpha=1
    ConstraintWeights_0=0.0003
    ConstraintWeights_1:float=0.1564
    ConstraintWeights_2:float=0.0813
    ConstraintWeights_3:float=0.0279
    ConstraintWeights_4:float=0.3046
    ConstraintWeights_5:float=0.0422
    ConstraintWeights_6:float=0.0452
    ConstraintWeights_7:float=0.021
    ConstraintWeights_8:float=0.3211
   ) -> (Pose) {
    alphaBoolBlend {blendOption=Linear}
    boneToModify {boneName=deform_headphones_base_r}
    constraintSetup [
     {targetBone {boneName=spine_04}}
     {targetBone {boneName=spine_05}}
     {targetBone {boneName=clavicle_r}}
     {targetBone {boneName=neck_01}}
     {targetBone {boneName=head}}
     {targetBone {boneName=neck_02}}
     {targetBone {boneName=deform_clavicle_r}}
     {targetBone {boneName=deform_collar_fr}}
     {targetBone {boneName=deform_collar_r}}
    ]
    constraintWeights [1,1,1,1,1,1,1,1,1]
   }
   n2 = "Constraint - deform_headphones_base_r_bk (PPPPPPPPP)"!(
    ComponentPose=@n1
    Alpha=1
    ConstraintWeights_0=0.0058
    ConstraintWeights_1:float=0.2169
    ConstraintWeights_2:float=0.0674
    ConstraintWeights_3:float=0.0369
    ConstraintWeights_4:float=0.1492
    ConstraintWeights_5:float=0.0706
    ConstraintWeights_6:float=0.0051
    ConstraintWeights_7:float=0.254
    ConstraintWeights_8:float=0.1941
   ) -> (Pose) {
    alphaBoolBlend {blendOption=Linear}
    boneToModify {boneName=deform_headphones_base_r_bk}
    constraintSetup [
     {targetBone {boneName=spine_03}}
     {targetBone {boneName=spine_05}}
     {targetBone {boneName=clavicle_r}}
     {targetBone {boneName=neck_01}}
     {targetBone {boneName=head}}
     {targetBone {boneName=neck_02}}
     {targetBone {boneName=deform_clavicle_r}}
     {targetBone {boneName=deform_collar_bk}}
     {targetBone {boneName=deform_collar_r}}
    ]
    constraintWeights [1,1,1,1,1,1,1,1,1]
   }
   n3 = "Constraint - deform_headphones_base_bk (PPPPPPPP)"!(
    ComponentPose=@n2
    Alpha=1
    ConstraintWeights_0=0.0146
    ConstraintWeights_1:float=0.2336
    ConstraintWeights_2:float=0.0037
    ConstraintWeights_3:float=0.0036
    ConstraintWeights_4:float=0.0563
    ConstraintWeights_5:float=0.1622
    ConstraintWeights_6:float=0.1039
    ConstraintWeights_7:float=0.4221
   ) -> (Pose) {
    alphaBoolBlend {blendOption=Linear}
    boneToModify {boneName=deform_headphones_base_bk}
    constraintSetup [
     {targetBone {boneName=spine_03}}
     {targetBone {boneName=spine_05}}
     {targetBone {boneName=clavicle_l}}
     {targetBone {boneName=clavicle_r}}
     {targetBone {boneName=neck_01}}
     {targetBone {boneName=head}}
     {targetBone {boneName=neck_02}}
     {targetBone {boneName=deform_collar_bk}}
    ]
    constraintWeights [1,1,1,1,1,1,1,1]
   }
   n4 = "Constraint - deform_headphones_base_l_bk (PPPPPPPP)"!(
    ComponentPose=@n3
    Alpha=1
    ConstraintWeights_0=0.0023
    ConstraintWeights_1:float=0.2109
    ConstraintWeights_2:float=0.0806
    ConstraintWeights_3:float=0.0332
    ConstraintWeights_4:float=0.2543
    ConstraintWeights_5:float=0.0649
    ConstraintWeights_6:float=0.1925
    ConstraintWeights_7:float=0.1614
   ) -> (Pose) {
    alphaBoolBlend {blendOption=Linear}
    boneToModify {boneName=deform_headphones_base_l_bk}
    constraintSetup [
     {targetBone {boneName=spine_03}}
     {targetBone {boneName=spine_05}}
     {targetBone {boneName=clavicle_l}}
     {targetBone {boneName=neck_01}}
     {targetBone {boneName=head}}
     {targetBone {boneName=neck_02}}
     {targetBone {boneName=deform_collar_bk}}
     {targetBone {boneName=deform_collar_l}}
    ]
    constraintWeights [1,1,1,1,1,1,1,1]
   }
   "Constraint - deform_headphones_base_l (PPPPPPPP)"!(
    ComponentPose=@n4
    Alpha=1
    ConstraintWeights_0=0.1547
    ConstraintWeights_1:float=0.0784
    ConstraintWeights_2:float=0.0276
    ConstraintWeights_3:float=0.3159
    ConstraintWeights_4:float=0.0416
    ConstraintWeights_5:float=0.0425
    ConstraintWeights_6:float=0.0223
    ConstraintWeights_7:float=0.3169
   ) -> (Pose) {
    alphaBoolBlend {blendOption=Linear}
    boneToModify {boneName=deform_headphones_base_l}
    constraintSetup [
     {targetBone {boneName=spine_05}}
     {targetBone {boneName=clavicle_l}}
     {targetBone {boneName=neck_01}}
     {targetBone {boneName=head}}
     {targetBone {boneName=neck_02}}
     {targetBone {boneName=deform_clavicle_l}}
     {targetBone {boneName=deform_collar_fr}}
     {targetBone {boneName=deform_collar_l}}
    ]
    constraintWeights [1,1,1,1,1,1,1,1]
   }
   n5 = Copy Pose From Mesh(SourceMeshComponent=Get Body() -> (Body:Skeletal Mesh Component Budgeted Object Reference)) {bUseAttachedParent=true}
   n6 = Copy Pose From Mesh() {bUseAttachedParent=true}
   n7 = Get HeadLimitAlpha() -> (HeadLimitAlpha:double)
   n8 = "Transform (Modify) Bone - Bone: head"(
    ComponentPose=Local To Component(LocalPose=Layered blend per bone(BasePose=@n6, BlendPoses_0=@n5) {layerSetup [{branchFilters [{boneName=clavicle_l}, {boneName=clavicle_r}]}]})
    Alpha=@n7
    Translation="0,0,0"
    Rotation="0,0,35"
    Scale="1,1,1"
   ) {
    rotationMode=BMM_Additive
    scaleMode=BMM_Additive
    translationMode=BMM_Additive
    alphaBoolBlend {blendOption=Linear}
    boneToModify {boneName=head}
    rotation {roll=35}
   }
   n9 = "Transform (Modify) Bone - Bone: neck_02"(ComponentPose=@n8, Alpha=@n7, Translation="0,0,0", Rotation="0,0,10", Scale="1,1,1") {
    rotationMode=BMM_Additive
    scaleMode=BMM_Additive
    translationMode=BMM_Additive
    alphaBoolBlend {blendOption=Linear}
    boneToModify {boneName=neck_02}
    rotation {roll=10}
   }
   n10 = "Transform (Modify) Bone - Bone: neck_01"(ComponentPose=@n9, Alpha=@n7, Translation="0,0,0", Rotation="0,0,5", Scale="1,1,1") {
    rotationMode=BMM_Additive
    scaleMode=BMM_Additive
    translationMode=BMM_Additive
    alphaBoolBlend {blendOption=Linear}
    boneToModify {boneName=neck_01}
    rotation {roll=5}
   }
   n11 = Control Rig!(
    Source=Component To Local(ComponentPose=@n10)
    Alpha
    HeadphonesMoveBackAlpha:double=0
    isPlayingEmoteYouBoreMe:bool=Get isYouBoreMe() -> (isYouBoreMe:bool)
   ) -> (Pose) {
    alphaBoolBlend {blendOption=Linear}
    controlRigAssetReference {blueprintRigClass=&/F_MED_BionicKitty_FaceAcc_CtrlRig.F_MED_BionicKitty_FaceAcc_CtrlRig_C}
   }
   Output Pose(Result=@n11) {layerGroup=None}
  ]}
 EventGraph {nodes [
   Event BlueprintUpdateAnimation()
   n1 = GetOwnerSkeletalMeshComponent()
   n2 = BreakRotator(InRot=BreakTransform(InTransform=GetDeltaTransformFromRefPose(self=@n1, BoneName=head, BaseName=neck_02)).Rotation)
   > Set HeadphonesMoveBackAlpha(HeadphonesMoveBackAlpha:double=MapRangeClamped(Value=@n2.Yaw, InRangeA=0, InRangeB=-40)) -> (Output_Get:double)
   > n3 = Branch(Condition=NOT Boolean(A=Get bIsPlayingEmote()))
   > Set HeadLimitAlpha(HeadLimitAlpha:double=MapRangeClamped(Value=@n2.Yaw, InRangeA=-10, InRangeB=-70)) -> (Output_Get:double)
   > Set isYouBoreMe(isYouBoreMe:bool=false) -> (Output_Get:bool)
   n4 = GetAnimInstance(self=GetOwnerSkeletalMeshComponent())
   Set isYouBoreMe(
    execute=@n3.else
    isYouBoreMe=OR Boolean!(
     A=Montage_IsPlaying(self=@n4, Montage=/BRCosmetics/Animation/Game/MainPlayer/Emotes/YouBoreMe/Emote_You_Bore_Me_CLF_M)
     B=Montage_IsPlaying(self=@n4, Montage=/BRCosmetics/Animation/Game/MainPlayer/Emotes/YouBoreMe/Emote_You_Bore_Me_CLM_M)
     C:bool=Montage_IsPlaying(self=@n4, Montage=/BRCosmetics/Animation/Game/MainPlayer/Emotes/YouBoreMe/Emote_You_Bore_Me_CMF_M)
     D:bool=Montage_IsPlaying(self=@n4, Montage=/BRCosmetics/Animation/Game/MainPlayer/Emotes/YouBoreMe/Emote_You_Bore_Me_CMM_M)
    ) -> (ReturnValue)
   )
   n5 = Event OnCharacterCustomizationLoaded()
   > Set Body(Body:Skeletal Mesh Component Budgeted Object Reference=Get Array Item with EFort Custom Part Type(Array=@n5.CustomMeshParts)) -> (Output_Get:Skeletal Mesh Component Budgeted Object Reference)
  ]}
]
variables [
 Body {type=SkeletalMeshComponentBudgeted}
 head rot {type=Rotator}
 HeadLimitAlpha {type=number}
 HeadphonesMoveBackAlpha {type=number}
 isYouBoreMe {type=boolean}
 X percentage {type=number, default=0.8}
 Y percentage {type=number, default=0.9}
 Z percentage {type=number, default=0.9}
]
